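# Sound cue for idle voice lines of the WildEstate ClimbSpill character, routed as critical enemy VO.
# A Random node picks among seven equally weighted wave players, each holding one idle VO recording.
# It uses shared henchmen NPC attenuation and boss combat chatter concurrency.
# Bus sends feed sidechain buses for music, weapon, first-person weapon and explosion volume.
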SoundCue /WildEstate_ClimbSpill/Assets/VO/IP_SC_VO_WildEstate_ClimbSpill_Idle
root=Random_1
nodes [
 {
  id=Random_1
  children [WavePlayer_5, WavePlayer_6, WavePlayer_7, WavePlayer_8, WavePlayer_9, WavePlayer_10, WavePlayer_11]
  properties {weights [1,1,1,1,1,1,1]}
 }
 {id=WavePlayer_10, properties {soundWaveAssetPtr=/WildEstate_ClimbSpill/Assets/VO/wav/vo_IP_WildEstate_ClimbSpill_Idle_022}}
 {id=WavePlayer_11, properties {soundWaveAssetPtr=/WildEstate_ClimbSpill/Assets/VO/wav/vo_IP_WildEstate_ClimbSpill_Idle_030}}
 {id=WavePlayer_5, properties {soundWaveAssetPtr=/WildEstate_ClimbSpill/Assets/VO/wav/vo_IP_WildEstate_ClimbSpill_Idle_010}}
 {id=WavePlayer_6, properties {soundWaveAssetPtr=/WildEstate_ClimbSpill/Assets/VO/wav/vo_IP_WildEstate_ClimbSpill_Idle_011}}
 {id=WavePlayer_7, properties {soundWaveAssetPtr=/WildEstate_ClimbSpill/Assets/VO/wav/vo_IP_WildEstate_ClimbSpill_Idle_012}}
 {id=WavePlayer_8, properties {soundWaveAssetPtr=/WildEstate_ClimbSpill/Assets/VO/wav/vo_IP_WildEstate_ClimbSpill_Idle_020}}
 {id=WavePlayer_9, properties {soundWaveAssetPtr=/WildEstate_ClimbSpill/Assets/VO/wav/vo_IP_WildEstate_ClimbSpill_Idle_021}}
]
properties {
 attenuationSettings=/Game/Sounds/Fort_Human_VO_BR/NPCs/Shared/NPC_VO_Henchmen_Att
 maxDistance=8500
 soundClassObject=&Fort_VO_Enemy_Critical
 volumeMultiplier=0.5957
 busSends [
  {audioBus=&AB_Sidechain_FortMusic_Vol, minSendLevel=0, sendLevel=1, soundSourceBus=null}
  {audioBus=/Game/Sounds/Modulation/EnvelopeFollowers/Volume/AB_Sidechain_FortWeapon1P_Vol, minSendLevel=0, sendLevel=1, soundSourceBus=null}
  {audioBus=&AB_Sidechain_FortWeapon_Vol, minSendLevel=0, sendLevel=1, soundSourceBus=null}
  {audioBus=&AB_Sidechain_FortExplosion_Vol, minSendLevel=0, sendLevel=1, soundSourceBus=null}
 ]
 concurrencySet [&NPC_Boss_Combat_Chatter_Conc]
}
